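# Billiard ball material for Niagara mesh particles, evidently for the Billiards skydiving contrail cosmetic.
# A random dynamic parameter picks ball color and variant: solid, striped, 8-ball or cue, with procedural UV spot and stripe masks.
# A scalar switch can force 8-ball only, and a view-space-normal reflection texture adds a fake matcap highlight.
# A CameraFade branch scales world normals by a negative strength through a pixel-depth sphere mask, shrinking balls near camera.
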Material /BRCosmetics/Athena/Effects/Materials/Skydiving/P_Contrails_Billiards/M_Billiards
graph {
 properties {bUsedWithNiagaraMeshParticles=true, pixelDepthOffsetMode=PDOM_Legacy, usedShadingModels=Default Lit}
 nodes [
  Constant()
  Constant() {r=1}
  n1 = PixelDepth()
  n2 = Constant()
  n3 = ScalarParameter() {defaultValue=128, group=CameraFade, parameterName=Radius}
  n4 = ScalarParameter() {group=CameraFade, parameterName=Hardness}
  Multiply(
   A=Multiply(A=SphereMask(A=@n1.R, B=@n2, Radius=@n3, Hardness=@n4), B=VertexNormalWS())
   B=ScalarParameter() {defaultValue=-50, group=CameraFade, parameterName=Strength}
  )
  n5 = Multiply(A=Transform(None=VertexNormalWS()) {transformSourceType=TRANSFORMSOURCE_World, transformType=TRANSFORM_View}) {constB=0.5}
  n6 = TextureSample(UVs=ComponentMask(None=Add(A=@n5) {constB=0.5}) {g=true, r=true}) {texture=/Game/VisualThreatManager/StormVisuals/LargeClouds/T_EyeballReflection2}
  n7 = ScalarParameter() {defaultValue=0.5, parameterName=Highlight Mult}
  MaterialExpressionDynamicParameter_0 = DynamicParameter() {defaultValue=<rgba 0.15 1 1 1>, paramNames [Random 0-1, na, na, na]}
  n8 = Custom!(None=@MaterialExpressionDynamicParameter_0.Random 0-1) {
   code="input = frac(input)*9;\r\ninput = floor(input);\r\n\r\nfloat3 _red = float3(.5,0,0);\r\nfloat3 _orange = float3(.8,.25,0);\r\nfloat3 _yellow = float3(.5,.5,0);\r\nfloat3 _green = float3(0,.5,0);\r\nfloat3 _blue = float3(0,0,.5);\r\nfloat3 _brown = float3(.22,.05,.05);\r\nfloat3 _violet = float3(.7,0,.7);\r\n\r\nfloat3 _colors[] = \r\n{\r\n    _red, _orange, _yellow, _green, _blue, _brown, _violet\r\n};\r\n\r\nuint input_i = (uint)input;\r\nif (input_i < 7)\r\n{\r\n     return _colors[input_i];\r\n}\r\nelse\r\n{\r\n     return _red;\r\n}"
   inputs [{
     inputName=input
     input {
      expression=&self:MaterialExpressionDynamicParameter_0
      mask=1
      maskR=1
     }
    }]
  }
  n9 = Constant3Vector() {constant=<rgba 0.6927 0.6332 0.534 0>}
  n10 = TextureCoordinate() {uTiling=2}
  n11 = Constant() {r=3}
  n12 = Divide(A=@n11)
  n13 = Multiply(A=Add(A=Multiply(A=@n10, B=@n11), B=Multiply(A=@n12) {constB=-1}) {constB=-1.5}) {constB=2}
  n14 = Multiply(A=@n13, B=@n13)
  n15 = Saturate(None=Add(A=ComponentMask(None=@n14) {r=true}, B=ComponentMask(None=@n14) {g=true}))
  n16 = LinearInterpolate(A=Constant3Vector() {constant=<rgba 0.005 0.005 0.005 0>}, B=@n9, Alpha=Ceil(None=OneMinus(None=@n15)))
  n17 = TextureCoordinate() {vTiling=0.5}
  n18 = Constant()
  n19 = If(A=Sine(None=ComponentMask(None=@n17) {g=true}), B=Constant() {r=0.92}, "A > B"=Constant() {r=1}, "A == B"=@n18, "A < B"=@n18)
  n20 = Frac(None=@MaterialExpressionDynamicParameter_0.Random 0-1)
  n21 = Round(None=Frac(None=Multiply(A=@n20) {constB=9}))
  n22 = Subtract(A=@MaterialExpressionDynamicParameter_0.Random 0-1, B=Constant() {r=0.7778})
  n23 = LinearInterpolate(
   A=LinearInterpolate(A=LinearInterpolate(A=@n8, B=@n9, Alpha=@n16), B=LinearInterpolate(A=@n9, B=@n8, Alpha=@n19), Alpha=@n21)
   B=LinearInterpolate(A=@n16, B=@n9, Alpha=@n21)
   Alpha=Saturate(None=Multiply(A=@n22) {constB=1000})
  )
  n24 = ScalarParameter() {group=CustomParams, parameterName="8BALL ONLY  OFF/ON", sliderMax=1, sliderMin=0}
  Saturate(None=Add(A=Multiply(A=@n6.RGB, B=@n7), B=LinearInterpolate(A=@n23, B=@n16, Alpha=@n24) {desc=8BALL ONLY LERP}))
 ]
}
